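# Unlit translucent flame material for Niagara mesh particles on the Season 23 VinderTech umbrella, tinted by an HDR color parameter.
# A panning normal texture distorts UVs for two scrolling noises, whose product is eroded along V, then masked by view-angle fade, a U-edge falloff and particle alpha.
# Dynamic parameters drive pan offset, speed and an apparent world-position offset weighted by inverted V.
# Both pan offset and displacement reuse their X parameter for Y.
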Material /BRCosmetics/Gadgets/Assets/VinderTech_Umbrella/Umbrella_Season23/FX/M_Umbrella_WPO_Flame
graph {
 properties {
  bUsedWithNiagaraMeshParticles=true
  blendMode=BLEND_Translucent
  pixelDepthOffsetMode=PDOM_Legacy
  shadingModel=MSM_Unlit
  usedShadingModels=Unlit
 }
 nodes [
  DynamicParameter() {parameterIndex=1, defaultValue=<rgba 0 0 1 1>, paramNames [Pan_Offset_X,Pan_Offset_Y,Speed,Intensity]}
  MaterialFunctionCall(Emissive=VectorParameter() {parameterName=Color, defaultValue=<rgba 47.75 27.5 50 1>}.RGB) {materialFunction=&MF_FX_EmissiveAdjustments}
  n1 = ComponentMask(None=TextureCoordinate()) {g=true}
  n2 = Add(A=Power(Base=OneMinus(None=@n1), Exp=ScalarParameter() {defaultValue=1, parameterName=Gradient Exponent})) {constB=0.4}
  n3 = Transform(None=Constant3Vector() {constant=<rgba 1 0 0 0>}) {transformSourceType=TRANSFORMSOURCE_Instance}
  n4 = ScalarParameter() {defaultValue=0.4, parameterName=Fresnel Exponent}
  Max(
   A=@n2
   B=OneMinus(None=MaterialFunctionCall(Axis=@n3, Exponent=@n4) {materialFunction=&AxisAlignedFresnel}.Fresnel)
  )
  n5 = Multiply(A=TextureCoordinate() {uTiling=2}, B=Constant2Vector() {g=0.125, r=0.125})
  n6 = DynamicParameter() {parameterIndex=1, defaultValue=<rgba 0 0 1 1>, paramNames [Pan_Offset_X,Pan_Offset_Y,Speed,Intensity]}
  n7 = MaterialFunctionCall(X=@n6.Pan_Offset_X, Y=@n6.Pan_Offset_X) {materialFunction=&MakeFloat2}
  n8 = TextureSample(UVs=Panner(Coordinate=Add(A=@n5, B=@n7.Result)) {speedX=-0.1, speedY=0.1}) {samplerType=SAMPLERTYPE_Normal, texture=/Game/Global/Textures/Default/Tileable_Patterns/1Normal}
  n9 = Add(
   A=TextureCoordinate()
   B=Multiply(A=MaterialFunctionCall(X=@n8.R, Y=@n8.R) {materialFunction=&MakeFloat2}.Result) {constB=0.1}
  )
  n10 = Multiply(A=Constant2Vector() {g=0.75, r=0.1}, B=@n6.Speed)
  n11 = TextureSample(UVs=Panner(Coordinate=Add(A=@n9, B=@n7.Result), Speed=@n10) {speedX=0.1, speedY=0.75}) {
   samplerType=SAMPLERTYPE_LinearGrayscale
   texture=/Game/Characters/Player/Male/Medium/Bodies/M_MED_Citadel/FX/T_Citadel_Noise_Dissolve_Handpaint
  }
  n12 = TextureSample(UVs=Panner(Coordinate=@n9, Speed=@n10) {speedX=0.1, speedY=0.75}) {texture=/Game/TimeOfDay/LightFunction/T_CloudNoise_Artemis}
  n13 = Multiply(A=@n11.R, B=Multiply(A=OneMinus(None=@n12.G)) {constB=1.2})
  n14 = ScalarParameter() {defaultValue=1.2, parameterName=Texture Exponent}
  n15 = Multiply(
   A=MaterialFunctionCall("FadeSensitivity (optional)"=Constant() {r=1.5}) {materialFunction=&FadeBasedOnViewAngle}.Result
   B=MaterialFunctionCall!(Gradient=Power(Base=@n13, Exp=@n14), Erosion Amount=@n2, Softness=ScalarParameter() {defaultValue=0.5, parameterName=Erosion Softness}) {materialFunction=/BRCosmetics/Athena/Effects/Particles/Skydiving/NS_Contrail_DarkAndLight/Functions/MF_Erosion}.Increased Range
  )
  n16 = ComponentMask(None=TextureCoordinate()) {r=true}
  n17 = OneMinus(None=@n16)
  n18 = Multiply(A=Power(Base=Multiply(A=@n16, B=@n17)) {constExponent=2.5}) {constB=5}
  Multiply(A=ParticleColor().A, B=Saturate(None=Multiply(A=@n15, B=@n18))) {constA=0.2}
  n19 = DynamicParameter() {defaultValue=<rgba 0 0 1 1>, paramNames [X, Y, "Z(unused)", Intensity]}
  n20 = TextureCoordinate()
  Multiply(
   A=MaterialFunctionCall(X=@n19.X, Y=@n19.X, Z=Constant()) {materialFunction=&MakeFloat3}.Result
   B=OneMinus(None=ComponentMask(None=@n20) {g=true})
  )
  TwoSidedSign()
 ]
}
