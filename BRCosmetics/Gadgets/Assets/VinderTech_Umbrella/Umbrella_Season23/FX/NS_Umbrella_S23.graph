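# Niagara effect for the Season 23 VinderTech umbrella, combining a glow sprite, plasma meshes, chest plasma and two ribbon trails from skeletal bones.
# Two system integrators blend the look, cross-fading paired plasma variants, stretching meshes and fading trails.
# Plasma writes a velocity-opposed WPO vector per particle, while chest plasma animates material parameters and scales alpha by time of day.
# Trails use separate front-end and gameplay velocity and alpha.
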NiagaraSystem /BRCosmetics/Gadgets/Assets/VinderTech_Umbrella/Umbrella_Season23/FX/NS_Umbrella_S23
dependencies {
 usedDataInterfaces [&NiagaraDataInterfaceCurve]
 usedDynamicInputs [&SetBoolByFloatComparison, &Add_Float, &Abs_Float]
 usedModules [&SystemState, &FN_Integrator]
}
emitters {
 Glow {
  emitterUpdateScript [
   &EmitterState(Delay First Loop Only=true, Inactive Response=::NewEnumerator1, Life Cycle Mode=::NewEnumerator1, Loop Delay=1, UseLoopDelay=true)
   &SpawnBurst_Instantaneous(Spawn Count=1, Spawn Probability=HlslExpression() {hlslExpression="Emitter.LoopCount == 0"}, Use Spawn Probability=true)
  ]
  particleSpawnScript [
   &InitializeParticle(
    Color=<rgba 0.955 0.55 1 0.2>
    Sprite Size Mode=::NewEnumerator3
    Uniform Sprite Size=100
    UsePositionOffset=false
    Write Material Random=true
   ) {module=InitializeParticle001}
   &SpriteFacingAndAlignment(Alignment Coordinate Space=::Local, Sprite Alignment=<1 0.25 0>, WriteAlignment=true, WriteFacing=false)
   &SystemLocation(MaskSpawnGroup=false, Offset=<0 0 -2>)
  ]
  particleUpdateScript [
   &ParticleState(
    Kill Particles When Lifetime Has Elapsed=false
    Let Infinitely Lived Particles Die When Emitter Deactivates=true
    Loop Particles Lifetime=true
   )
   &CameraOffset(Camera Offset Amount=1)
   &ScaleColor(
    Scale Alpha=&FloatFromCurve(
     FloatCurve=DataInterface() {propertyValues {Curve {keys [
         {arriveTangent=-0.4837, interpMode=RCIM_Cubic, leaveTangent=-0.4837, tangentMode=RCTM_User, time=0, value=1}
         {arriveTangent=0.862, interpMode=RCIM_Cubic, leaveTangent=0.862, tangentMode=RCTM_User, time=0.5, value=1.025}
         {arriveTangent=0.8247, interpMode=RCIM_Cubic, leaveTangent=0.8247, tangentMode=RCTM_User, time=1, value=1}
        ]}}}
     Scale Curve=&Lerp_Float(A=2, B=1, Alpha=$System.FN_Integrator001.NormalizedValue)
    )
    ScaleA=true
    ScaleRGB=true
   )
  ]
  rendererClasses [&NiagaraSpriteRendererProperties]
  renderers [{rendererClass=&NiagaraSpriteRendererProperties}]
 }
 Plasma {
  emitterUpdateScript [
   &EmitterState(Delay First Loop Only=true, Life Cycle Mode=::NewEnumerator1, Loop Delay=1, Loop Duration=2, UseLoopDelay=true)
   &SpawnBurst_Instantaneous(Spawn Probability=HlslExpression() {hlslExpression="Emitter.LoopCount == 0"}, Use Spawn Probability=true, UseLoopCountLimit=false)
  ]
  particleSpawnScript [
   &InitializeParticle(Mesh Scale=<3.25 6.5 6.5>, Mesh Scale Mode=::NewEnumerator4, UsePositionOffset=false, Write Material Random=true)
   &CameraOffset(Camera Offset Amount=2.5)
   &InitialMeshOrientation(
    Mesh Axis To Orient=<0 0 -1>
    Mesh Orientation Mode=::NewEnumerator2
    Orientation Coordinate Space=::Local
    Rotation=<0.5 0 0>
    Use Rotation Vector=true
   )
   &DynamicMaterialParameters(
    Index 1 Param 1=&RandomRangeFloat(Minimum=-0.333, Maximum=-0.666)
    Index 1 Param 2=&RandomRangeFloat(Minimum=0.333, Maximum=0.666)
    Index 1 Param 3=1
    Module.Param0WriteEnabled1=true
    Module.Param1WriteEnabled1=true
    Module.Param2WriteEnabled1=true
    Module.Param3WriteEnabled1=false
    Write Parameter Index 0=false
    Write Parameter Index 1=true
   )
   &SystemLocation(MaskSpawnGroup=false, Offset=<0 0 -2.5>)
  ]
  particleUpdateScript [
   &ParticleState(Kill Particles When Lifetime Has Elapsed=false, Let Infinitely Lived Particles Die When Emitter Deactivates=true)
   &ScaleMeshSize(Scale Factor=&MakeVector(
     X=&FloatFromCurve(
      FloatCurve=DataInterface() {propertyValues {Curve {keys [
          {arriveTangent=4.1785, interpMode=RCIM_Cubic, leaveTangent=4.1785, tangentMode=RCTM_User, time=0, value=0}
          {arriveTangent=1.1179, interpMode=RCIM_Cubic, leaveTangent=1.1179, tangentMode=RCTM_User, time=0.2042, value=0.6228}
          {arriveTangent=0.0224, interpMode=RCIM_Cubic, leaveTangent=0.0224, tangentMode=RCTM_User, time=1, value=1}
         ]}}}
      Scale Curve=&Lerp_Float(A=1.5, B=0.75, Alpha=$System.FN_Integrator.NormalizedValue)
     )
     Y=1
     Z=1
    ))
   /SetVariables(Particles.WPO:Vector4f=&MakeFloat4FromVectorAndFloat(
     "Vector (RGB)"=&Multiply_VectorByFloat(
      Vector=&NormalizeVector(Vector To Normalize=$Engine.Owner.Velocity, Safe Normalize=true, Safe Fallback Threshold=0, Safe Fallback Vector=<0 0 0.001>)
      Float=&Lerp_Float(
       A=&Lerp_Float(A=-150, B=&RandomRangeFloat(Minimum=-80, Maximum=-100), Alpha=$System.FN_Integrator001.NormalizedValue)
       B=0
       Alpha=$System.FN_Integrator.NormalizedValue
      )
     )
     "Float (Alpha)"=1
    )) {isSetParameters=true, module=SetVariables_#1}
   &Color(Color=&MakeLinearColorFromVectorAndFloat(
     "Vector (RGB)"=<1 1 1>
     "Float (Alpha)"=&FloatFromCurve(
      FloatCurve=DataInterface() {propertyValues {Curve {keys [
          {interpMode=RCIM_Cubic, tangentMode=RCTM_User, time=0, value=0}
          {arriveTangent=2.4448, interpMode=RCIM_Cubic, leaveTangent=2.4448, tangentMode=RCTM_User, time=1, value=1}
         ]}}}
      Scale Curve=&Lerp_Float(A=0, B=1, Alpha=$System.FN_Integrator001.NormalizedValue)
     )
    ))
  ]
  rendererClasses [&NiagaraMeshRendererProperties]
  renderers [{rendererClass=&NiagaraMeshRendererProperties, rendererIndex=0}]
 }
 Plasma001 {
  emitterUpdateScript [
   &EmitterState(Delay First Loop Only=true, Life Cycle Mode=::NewEnumerator1, Loop Delay=1, Loop Duration=2, UseLoopDelay=true)
   &SpawnBurst_Instantaneous(Spawn Probability=HlslExpression() {hlslExpression="Emitter.LoopCount == 0"}, Use Spawn Probability=true, UseLoopCountLimit=false)
  ]
  particleSpawnScript [
   &InitializeParticle(Mesh Scale=<3.25 6.5 6.5>, Mesh Scale Mode=::NewEnumerator4, UsePositionOffset=false, Write Material Random=true)
   &CameraOffset(Camera Offset Amount=2.5)
   &InitialMeshOrientation(
    Mesh Axis To Orient=<0 0 -1>
    Mesh Orientation Mode=::NewEnumerator2
    Orientation Coordinate Space=::Local
    Rotation=<0.5 0 0>
    Use Rotation Vector=true
   )
   &DynamicMaterialParameters(
    Index 1 Param 1=&RandomRangeFloat(Minimum=-0.333, Maximum=-0.666)
    Index 1 Param 2=&RandomRangeFloat(Minimum=0.333, Maximum=0.666)
    Index 1 Param 3=2
    Module.Param0WriteEnabled1=true
    Module.Param1WriteEnabled1=true
    Module.Param2WriteEnabled1=true
    Module.Param3WriteEnabled1=false
    Write Parameter Index 0=false
    Write Parameter Index 1=true
   )
   &SystemLocation(MaskSpawnGroup=false, Offset=<0 0 -2.5>)
  ]
  particleUpdateScript [
   &ParticleState(Kill Particles When Lifetime Has Elapsed=false, Let Infinitely Lived Particles Die When Emitter Deactivates=true)
   &ScaleMeshSize(Scale Factor=&MakeVector(
     X=&FloatFromCurve(
      FloatCurve=DataInterface() {propertyValues {Curve {keys [
          {arriveTangent=4.1785, interpMode=RCIM_Cubic, leaveTangent=4.1785, tangentMode=RCTM_User, time=0, value=0}
          {arriveTangent=1.1179, interpMode=RCIM_Cubic, leaveTangent=1.1179, tangentMode=RCTM_User, time=0.2042, value=0.6228}
          {arriveTangent=0.0224, interpMode=RCIM_Cubic, leaveTangent=0.0224, tangentMode=RCTM_User, time=1, value=1}
         ]}}}
      Scale Curve=&Lerp_Float(A=1.5, B=0.75, Alpha=$System.FN_Integrator.NormalizedValue)
     )
     Y=1
     Z=1
    ))
   /SetVariables(Particles.WPO:Vector4f=&MakeFloat4FromVectorAndFloat(
     "Vector (RGB)"=&Multiply_VectorByFloat(
      Vector=&NormalizeVector(Vector To Normalize=$Engine.Owner.Velocity, Safe Normalize=true, Safe Fallback Threshold=0, Safe Fallback Vector=<0 0 0.001>)
      Float=&Lerp_Float(
       A=&Lerp_Float(A=-150, B=&RandomRangeFloat(Minimum=-80, Maximum=-100), Alpha=$System.FN_Integrator001.NormalizedValue)
       B=0
       Alpha=$System.FN_Integrator.NormalizedValue
      )
     )
     "Float (Alpha)"=1
    )) {isSetParameters=true, module=SetVariables_#1}
   &Color(Color=&MakeLinearColorFromVectorAndFloat(
     "Vector (RGB)"=<1 1 1>
     "Float (Alpha)"=&FloatFromCurve(
      FloatCurve=DataInterface() {propertyValues {Curve {keys [
          {interpMode=RCIM_Cubic, tangentMode=RCTM_User, time=0, value=0}
          {arriveTangent=2.4448, interpMode=RCIM_Cubic, leaveTangent=2.4448, tangentMode=RCTM_User, time=1, value=1}
         ]}}}
      Scale Curve=&Lerp_Float(A=1, B=0, Alpha=$System.FN_Integrator001.NormalizedValue)
     )
    ))
  ]
  rendererClasses [&NiagaraMeshRendererProperties]
  renderers [{rendererClass=&NiagaraMeshRendererProperties, rendererIndex=0}]
 }
 Plasma_Chest {
  emitterUpdateScript [
   &EmitterState(Delay First Loop Only=true, Life Cycle Mode=::NewEnumerator1, Loop Delay=1, UseLoopDelay=true)
   &SpawnBurst_Instantaneous(Spawn Probability=HlslExpression() {hlslExpression="Emitter.LoopCount == 0"}, Use Spawn Probability=true, UseLoopCountLimit=false)
  ]
  particleSpawnScript [
   &InitializeParticle(Mesh Scale=<2.5 2.5 2.5>, Mesh Scale Mode=::NewEnumerator4, UsePositionOffset=false, Write Material Random=true)
   &InitialMeshOrientation(Mesh Axis To Orient=<0 0 -1>, Mesh Orientation Mode=::NewEnumerator2, Rotation=<0.5 0 0>, Use Rotation Vector=true)
   &SystemLocation(MaskSpawnGroup=false, Offset=<2.5 0 -4>)
  ]
  particleUpdateScript [
   &ParticleState(
    Kill Particles When Lifetime Has Elapsed=false
    Let Infinitely Lived Particles Die When Emitter Deactivates=true
    Loop Particles Lifetime=true
   )
   &CameraOffset(Camera Offset Amount=2.5)
   &DynamicMaterialParameters(
    Index 0 Param 1=&FloatFromCurve(
     FloatCurve=DataInterface() {propertyValues {Curve {keys [
         {interpMode=RCIM_Cubic, tangentMode=RCTM_User, time=0, value=0.05}
         {arriveTangent=-0.0008, interpMode=RCIM_Cubic, leaveTangent=-0.0008, tangentMode=RCTM_User, time=0.1474, value=0.0316}
         {arriveTangent=-0.0621, interpMode=RCIM_Cubic, leaveTangent=-0.0621, tangentMode=RCTM_User, time=0.2888, value=0.1386}
         {arriveTangent=-0.1732, interpMode=RCIM_Cubic, leaveTangent=-0.1732, tangentMode=RCTM_User, time=0.4007, value=0.05}
         {interpMode=RCIM_Cubic, tangentMode=RCTM_User, time=0.4826, value=0.0427}
         {arriveTangent=-0.0057, interpMode=RCIM_Cubic, leaveTangent=-0.0057, tangentMode=RCTM_User, time=0.593, value=0.037}
         {arriveTangent=0.0385, interpMode=RCIM_Cubic, leaveTangent=0.0385, time=0.8161, value=0.1381}
         {arriveTangent=-0.0113, interpMode=RCIM_Cubic, leaveTangent=-0.0113, tangentMode=RCTM_User, time=0.9311, value=0.05}
         {tangentMode=RCTM_User, time=1, value=0.05}
        ]}}}
     CurveIndex=$Emitter.NormalizedLoopAge
    )
    Index 0 Param 2=&FloatFromCurve(
     FloatCurve=DataInterface() {propertyValues {Curve {keys [
         {interpMode=RCIM_Cubic, tangentMode=RCTM_User, time=0, value=150}
         {arriveTangent=-0.0008, interpMode=RCIM_Cubic, leaveTangent=-0.0008, tangentMode=RCTM_User, time=0.0479, value=150}
         {arriveTangent=174.1042, interpMode=RCIM_Cubic, leaveTangent=174.1042, time=0.1648, value=200}
         {arriveTangent=-484.7725, interpMode=RCIM_Cubic, leaveTangent=-484.7725, tangentMode=RCTM_User, time=0.2573, value=186.446}
         {interpMode=RCIM_Cubic, tangentMode=RCTM_User, time=0.3963, value=230}
         {arriveTangent=372.7463, interpMode=RCIM_Cubic, leaveTangent=372.7463, tangentMode=RCTM_User, time=0.6102, value=165.0835}
         {arriveTangent=-82.9196, interpMode=RCIM_Cubic, leaveTangent=-82.9196, time=0.7076, value=200}
         {arriveTangent=-0.0113, interpMode=RCIM_Cubic, leaveTangent=-0.0113, tangentMode=RCTM_User, time=0.7921, value=150}
         {tangentMode=RCTM_User, time=1, value=150}
        ]}}}
     Scale Curve=0.5
    )
    Module.Param0WriteEnabled=true
    Module.Param1WriteEnabled=true
    Module.Param2WriteEnabled=false
    Module.Param3WriteEnabled=false
   )
   &ScaleMeshSize(Scale Factor=&MakeVector(
     X=&FloatFromCurve(FloatCurve=DataInterface() {propertyValues {Curve {keys [
          {interpMode=RCIM_Cubic, tangentMode=RCTM_User, time=0, value=0.95}
          {interpMode=RCIM_Cubic, tangentMode=RCTM_User, time=0.5, value=1}
          {tangentMode=RCTM_User, time=1, value=0.95}
         ]}}})
     Y=&FloatFromCurve(FloatCurve=DataInterface() {propertyValues {Curve {keys [
          {interpMode=RCIM_Cubic, tangentMode=RCTM_User, time=0, value=0.95}
          {interpMode=RCIM_Cubic, tangentMode=RCTM_User, time=0.5, value=1}
          {tangentMode=RCTM_User, time=1, value=0.95}
         ]}}})
     Z=1
    ))
   &Color(Scale Alpha=&Lerp_Float(
     A=&FNTimeOfDayScalar(Night Scale=1.5, Morning Scale=2, Day Scale=5, Evening Scale=1.75)
     B=&FNTimeOfDayScalar(Night Scale=0.5, Morning Scale=0.5, Day Scale=0.5, Evening Scale=0.5)
     Alpha=$System.FN_Integrator001.NormalizedValue
    ))
  ]
  rendererClasses [&NiagaraMeshRendererProperties]
  renderers [{rendererClass=&NiagaraMeshRendererProperties, rendererIndex=0}]
 }
 Plasma_Chest001 {
  emitterUpdateScript [
   &EmitterState(Delay First Loop Only=true, Life Cycle Mode=::NewEnumerator1, Loop Delay=1, UseLoopDelay=true)
   &SpawnBurst_Instantaneous(Spawn Probability=HlslExpression() {hlslExpression="Emitter.LoopCount == 0"}, Use Spawn Probability=true, UseLoopCountLimit=false)
  ]
  particleSpawnScript [
   &InitializeParticle(Mesh Scale Mode=::NewEnumerator3, Mesh Uniform Scale=1.1, UsePositionOffset=false, Write Material Random=true)
   &InitialMeshOrientation(Mesh Orientation Mode=::NewEnumerator2, Use Rotation Vector=false)
   &SystemLocation(MaskSpawnGroup=false, Offset=<1 -1 2>)
  ]
  particleUpdateScript [
   &ParticleState(
    Kill Particles When Lifetime Has Elapsed=false
    Let Infinitely Lived Particles Die When Emitter Deactivates=true
    Loop Particles Lifetime=true
   )
   &CameraOffset(Camera Offset Amount=2)
   &DynamicMaterialParameters(
    Index 0 Param 1=&FloatFromCurve(
     FloatCurve=DataInterface() {propertyValues {Curve {keys [
         {interpMode=RCIM_Cubic, tangentMode=RCTM_User, time=0, value=0.05}
         {arriveTangent=-0.0008, interpMode=RCIM_Cubic, leaveTangent=-0.0008, tangentMode=RCTM_User, time=0.1474, value=0.0316}
         {arriveTangent=-0.0621, interpMode=RCIM_Cubic, leaveTangent=-0.0621, tangentMode=RCTM_User, time=0.2888, value=0.1386}
         {arriveTangent=-0.1732, interpMode=RCIM_Cubic, leaveTangent=-0.1732, tangentMode=RCTM_User, time=0.4007, value=0.05}
         {interpMode=RCIM_Cubic, tangentMode=RCTM_User, time=0.4826, value=0.0427}
         {arriveTangent=-0.0057, interpMode=RCIM_Cubic, leaveTangent=-0.0057, tangentMode=RCTM_User, time=0.593, value=0.037}
         {arriveTangent=0.0385, interpMode=RCIM_Cubic, leaveTangent=0.0385, time=0.8161, value=0.1381}
         {arriveTangent=-0.0113, interpMode=RCIM_Cubic, leaveTangent=-0.0113, tangentMode=RCTM_User, time=0.9311, value=0.05}
         {tangentMode=RCTM_User, time=1, value=0.05}
        ]}}}
     CurveIndex=$Emitter.NormalizedLoopAge
    )
    Index 0 Param 2=&FloatFromCurve(
     FloatCurve=DataInterface() {propertyValues {Curve {keys [
         {interpMode=RCIM_Cubic, tangentMode=RCTM_User, time=0, value=150}
         {arriveTangent=-0.0008, interpMode=RCIM_Cubic, leaveTangent=-0.0008, tangentMode=RCTM_User, time=0.0479, value=150}
         {arriveTangent=174.1042, interpMode=RCIM_Cubic, leaveTangent=174.1042, time=0.1648, value=200}
         {arriveTangent=-484.7725, interpMode=RCIM_Cubic, leaveTangent=-484.7725, tangentMode=RCTM_User, time=0.2573, value=186.446}
         {interpMode=RCIM_Cubic, tangentMode=RCTM_User, time=0.3963, value=230}
         {arriveTangent=372.7463, interpMode=RCIM_Cubic, leaveTangent=372.7463, tangentMode=RCTM_User, time=0.6102, value=165.0835}
         {arriveTangent=-82.9196, interpMode=RCIM_Cubic, leaveTangent=-82.9196, time=0.7076, value=200}
         {arriveTangent=-0.0113, interpMode=RCIM_Cubic, leaveTangent=-0.0113, tangentMode=RCTM_User, time=0.7921, value=150}
         {tangentMode=RCTM_User, time=1, value=150}
        ]}}}
     Scale Curve=0.5
    )
    Module.Param0WriteEnabled=true
    Module.Param1WriteEnabled=true
    Module.Param2WriteEnabled=false
    Module.Param3WriteEnabled=false
   )
   &ScaleMeshSize(Scale Factor=&MakeVector(
     X=&FloatFromCurve(FloatCurve=DataInterface() {propertyValues {Curve {keys [
          {interpMode=RCIM_Cubic, tangentMode=RCTM_User, time=0, value=0.95}
          {interpMode=RCIM_Cubic, tangentMode=RCTM_User, time=0.5, value=1}
          {tangentMode=RCTM_User, time=1, value=0.95}
         ]}}})
     Y=&FloatFromCurve(FloatCurve=DataInterface() {propertyValues {Curve {keys [
          {interpMode=RCIM_Cubic, tangentMode=RCTM_User, time=0, value=0.95}
          {interpMode=RCIM_Cubic, tangentMode=RCTM_User, time=0.5, value=1}
          {tangentMode=RCTM_User, time=1, value=0.95}
         ]}}})
     Z=1
    ))
   &Color(Color=&MakeLinearColorFromVectorAndFloat(
     "Vector (RGB)"=<1 1 1>
     "Float (Alpha)"=&Lerp_Float(
      A=&FNTimeOfDayScalar(Night Scale=1.75, Morning Scale=2.5, Day Scale=5, Evening Scale=1.75)
      B=&FNTimeOfDayScalar(Night Scale=0.5, Morning Scale=0.5, Day Scale=0.5, Evening Scale=0.5)
      Alpha=$System.FN_Integrator001.NormalizedValue
     )
    ))
  ]
  rendererClasses [&NiagaraMeshRendererProperties]
  renderers [{rendererClass=&NiagaraMeshRendererProperties, rendererIndex=0}]
 }
 Plasma_Inverted {
  emitterUpdateScript [
   &EmitterState(Delay First Loop Only=true, Life Cycle Mode=::NewEnumerator1, Loop Delay=1, Loop Duration=0.5, UseLoopDelay=true)
   &SpawnBurst_Instantaneous(Spawn Probability=HlslExpression() {hlslExpression="Emitter.LoopCount == 0"}, Use Spawn Probability=true, UseLoopCountLimit=false)
  ]
  particleSpawnScript [
   &InitializeParticle(Mesh Scale=<3.25 6.5 6.5>, Mesh Scale Mode=::NewEnumerator4, UsePositionOffset=false, Write Material Random=true)
   &CameraOffset(Camera Offset Amount=2.5)
   &InitialMeshOrientation(
    Mesh Axis To Orient=<0 0 -1>
    Mesh Orientation Mode=::NewEnumerator2
    Orientation Coordinate Space=::Local
    Rotation=<0.5 0 0>
    Use Rotation Vector=true
   )
   &DynamicMaterialParameters(
    Index 1 Param 1=&RandomRangeFloat(Minimum=0, Maximum=-0.5)
    Index 1 Param 2=&RandomRangeFloat(Minimum=0, Maximum=0.5)
    Index 1 Param 3=1
    Module.Param0WriteEnabled1=true
    Module.Param1WriteEnabled1=true
    Module.Param2WriteEnabled1=true
    Module.Param3WriteEnabled1=false
    Write Parameter Index 0=false
    Write Parameter Index 1=true
   )
   &SystemLocation(MaskSpawnGroup=false, Offset=<0 0 -2.5>)
  ]
  particleUpdateScript [
   &ParticleState(Kill Particles When Lifetime Has Elapsed=false, Let Infinitely Lived Particles Die When Emitter Deactivates=true)
   &ScaleMeshSize(Scale Factor=&MakeVector(
     X=&FloatFromCurve(
      FloatCurve=DataInterface() {propertyValues {Curve {keys [
          {arriveTangent=4.1785, interpMode=RCIM_Cubic, leaveTangent=4.1785, tangentMode=RCTM_User, time=0, value=0}
          {arriveTangent=1.1179, interpMode=RCIM_Cubic, leaveTangent=1.1179, tangentMode=RCTM_User, time=0.2042, value=0.6228}
          {arriveTangent=0.0224, interpMode=RCIM_Cubic, leaveTangent=0.0224, tangentMode=RCTM_User, time=1, value=1}
         ]}}}
      Scale Curve=&Lerp_Float(A=1.5, B=0.75, Alpha=$System.FN_Integrator.NormalizedValue)
     )
     Y=1
     Z=1
    ))
   /SetVariables(Particles.WPO:Vector4f=&MakeFloat4FromVectorAndFloat(
     "Vector (RGB)"=&Multiply_VectorByFloat(
      Vector=&NormalizeVector(Vector To Normalize=$Engine.Owner.Velocity, Safe Normalize=true, Safe Fallback Threshold=0, Safe Fallback Vector=<0 0 0.001>)
      Float=&Lerp_Float(
       A=&Lerp_Float(A=-150, B=&RandomRangeFloat(Minimum=-75, Maximum=-100), Alpha=$System.FN_Integrator001.NormalizedValue)
       B=0
       Alpha=$System.FN_Integrator.NormalizedValue
      )
     )
     "Float (Alpha)"=1
    )) {isSetParameters=true, module=SetVariables_#1}
   &Color(Color=&MakeLinearColorFromVectorAndFloat(
     "Vector (RGB)"=<1 1 1>
     "Float (Alpha)"=&FloatFromCurve(
      FloatCurve=DataInterface() {propertyValues {Curve {keys [
          {interpMode=RCIM_Cubic, tangentMode=RCTM_User, time=0, value=0}
          {arriveTangent=2.4448, interpMode=RCIM_Cubic, leaveTangent=2.4448, tangentMode=RCTM_User, time=1, value=1}
         ]}}}
      Scale Curve=&Lerp_Float(A=0, B=1, Alpha=$System.FN_Integrator001.NormalizedValue)
     )
    ))
  ]
  rendererClasses [&NiagaraMeshRendererProperties]
  renderers [{rendererClass=&NiagaraMeshRendererProperties, rendererIndex=0}]
 }
 Plasma_Inverted001 {
  emitterUpdateScript [
   &EmitterState(Delay First Loop Only=true, Life Cycle Mode=::NewEnumerator1, Loop Delay=1, Loop Duration=0.5, UseLoopDelay=true)
   &SpawnBurst_Instantaneous(Spawn Probability=HlslExpression() {hlslExpression="Emitter.LoopCount == 0"}, Use Spawn Probability=true, UseLoopCountLimit=false)
  ]
  particleSpawnScript [
   &InitializeParticle(Mesh Scale=<3.25 6.5 6.5>, Mesh Scale Mode=::NewEnumerator4, UsePositionOffset=false, Write Material Random=true)
   &CameraOffset(Camera Offset Amount=2.5)
   &InitialMeshOrientation(
    Mesh Axis To Orient=<0 0 -1>
    Mesh Orientation Mode=::NewEnumerator2
    Orientation Coordinate Space=::Local
    Rotation=<0.5 0 0>
    Use Rotation Vector=true
   )
   &DynamicMaterialParameters(
    Index 1 Param 1=&RandomRangeFloat(Minimum=0, Maximum=-0.5)
    Index 1 Param 2=&RandomRangeFloat(Minimum=0, Maximum=0.5)
    Index 1 Param 3=2
    Module.Param0WriteEnabled1=true
    Module.Param1WriteEnabled1=true
    Module.Param2WriteEnabled1=true
    Module.Param3WriteEnabled1=false
    Write Parameter Index 0=false
    Write Parameter Index 1=true
   )
   &SystemLocation(MaskSpawnGroup=false, Offset=<0 0 -2.5>)
  ]
  particleUpdateScript [
   &ParticleState(Kill Particles When Lifetime Has Elapsed=false, Let Infinitely Lived Particles Die When Emitter Deactivates=true)
   &ScaleMeshSize(Scale Factor=&MakeVector(
     X=&FloatFromCurve(
      FloatCurve=DataInterface() {propertyValues {Curve {keys [
          {arriveTangent=4.1785, interpMode=RCIM_Cubic, leaveTangent=4.1785, tangentMode=RCTM_User, time=0, value=0}
          {arriveTangent=1.1179, interpMode=RCIM_Cubic, leaveTangent=1.1179, tangentMode=RCTM_User, time=0.2042, value=0.6228}
          {arriveTangent=0.0224, interpMode=RCIM_Cubic, leaveTangent=0.0224, tangentMode=RCTM_User, time=1, value=1}
         ]}}}
      Scale Curve=&Lerp_Float(A=1.5, B=0.75, Alpha=$System.FN_Integrator.NormalizedValue)
     )
     Y=1
     Z=1
    ))
   /SetVariables(Particles.WPO:Vector4f=&MakeFloat4FromVectorAndFloat(
     "Vector (RGB)"=&Multiply_VectorByFloat(
      Vector=&NormalizeVector(Vector To Normalize=$Engine.Owner.Velocity, Safe Normalize=true, Safe Fallback Threshold=0, Safe Fallback Vector=<0 0 0.001>)
      Float=&Lerp_Float(
       A=&Lerp_Float(A=-150, B=&RandomRangeFloat(Minimum=-75, Maximum=-100), Alpha=$System.FN_Integrator001.NormalizedValue)
       B=0
       Alpha=$System.FN_Integrator.NormalizedValue
      )
     )
     "Float (Alpha)"=1
    )) {isSetParameters=true, module=SetVariables_#1}
   &Color(Color=&MakeLinearColorFromVectorAndFloat(
     "Vector (RGB)"=<1 1 1>
     "Float (Alpha)"=&FloatFromCurve(
      FloatCurve=DataInterface() {propertyValues {Curve {keys [
          {interpMode=RCIM_Cubic, tangentMode=RCTM_User, time=0, value=0}
          {arriveTangent=2.4448, interpMode=RCIM_Cubic, leaveTangent=2.4448, tangentMode=RCTM_User, time=1, value=1}
         ]}}}
      Scale Curve=&Lerp_Float(A=1, B=0, Alpha=$System.FN_Integrator001.NormalizedValue)
     )
    ))
  ]
  rendererClasses [&NiagaraMeshRendererProperties]
  renderers [{rendererClass=&NiagaraMeshRendererProperties, rendererIndex=0}]
 }
 Trails {
  emitterUpdateScript [
   &EmitterState(Life Cycle Mode=::NewEnumerator1, Loop Duration=1, UseLoopDelay=false)
   /Game/Effects/Niagara/Modules/AnimTrail/FN_SpawnAnimTrail/FN_SpawnAnimTrail(Number of Anim Trails=2, SpawnRate=64)
  ]
  particleSpawnScript [
   /Game/Effects/Niagara/Modules/AnimTrail/FN_InitializeAnimTrail/FN_InitializeAnimTrailBoneIndecies(First Bone or Socket Index=1)
   &FN_InitializeAnimTrailBoneIndecies(AnimTrail Selection=1, First Bone or Socket Index=3, Second Bone or Socket Index=2) {module=FN_InitializeAnimTrailBoneIndecies001}
   /Game/Effects/Niagara/Modules/AnimTrail/FN_InitializeAnimTrail/FN_InitializeAnimTrail(AnimTrailWidth=1, Color=<rgba 0.6 0.6 0.6 0.8>, Lifetime=0.25)
   &AddVelocity(Velocity=&FN_FrontEndVector3Modifier(
     Gameplay=&Lerp_Vector(A=<0 0 0>, B=<0 0 400>, Alpha=&OneMinusFloat(Float=$System.FN_Integrator001.Value))
     Front End=<0 2000 0>
    ))
   &AddVelocity(Velocity=&FN_FrontEndVector3Modifier(Gameplay=<0 0 0>, Front End=<0 0 0>)) {enabled=false, module=AddVelocity001}
  ]
  particleUpdateScript [
   &InheritSourceMovement(Applied Position Scale=<0.25 0.25 0.8>, ApplyForce=false, ApplyPosition=true, ApplyPositionFalloff=false)
   &SolveForcesAndVelocity(Clamp Velocity=false, Limit Acceleration=false)
   /Game/Effects/Niagara/Modules/AnimTrail/FN_UpdateAnimTrail/FN_UpdateAnimTrail(New float=0.8)
   &ScaleColor(
    Scale Alpha=&FN_FrontEndFloatModifier(Gameplay=&OneMinusFloat(Float=$System.FN_Integrator001.Value), Front End=1)
    ScaleA=true
    ScaleRGB=true
   )
  ]
  rendererClasses [&NiagaraRibbonRendererProperties]
  renderers [{rendererClass=&NiagaraRibbonRendererProperties}]
 }
}
summary {
 systemName=NS_Umbrella_S23
 emitters [
  {emitterName=Plasma_Inverted, rendererClasses [&NiagaraMeshRendererProperties]}
  {emitterName=Plasma, rendererClasses [&NiagaraMeshRendererProperties]}
  {emitterName=Plasma_Chest, rendererClasses [&NiagaraMeshRendererProperties]}
  {emitterName=Plasma_Chest001, rendererClasses [&NiagaraMeshRendererProperties]}
  {emitterName=Glow, rendererClasses [&NiagaraSpriteRendererProperties]}
  {emitterName=Trails, rendererClasses [&NiagaraRibbonRendererProperties]}
  {emitterName=Plasma_Inverted001, rendererClasses [&NiagaraMeshRendererProperties]}
  {emitterName=Plasma001, rendererClasses [&NiagaraMeshRendererProperties]}
 ]
 userVariables [
  User.bIsFrontEnd {type=NiagaraBool}
  User.bIsFrontEndPreview {type=NiagaraBool}
  User.bIsFullyDeployed {type=NiagaraBool}
  User.ForwardDot {type=NiagaraFloat, defaultValue=0, description=""}
  User.RibbonIDOffset {type=NiagaraInt32}
  User.RightDot {type=NiagaraFloat, defaultValue=0, description=""}
  User.SkeletalMesh {
   type=NiagaraDataInterfaceSkeletalMesh
   defaultValue {
    struct=NiagaraExt_VariableValue_DataInterface
    value {dataInterface=&self:NiagaraDataInterfaceSkeletalMesh_0}
   }
  }
 ]
}
